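# Level Sequence for the Season 12 battle pass Photographer intro cinematic in the frontend menu.
# Numbered cine camera shots, some keyed to pan or tilt, frame two animated mannequins and gadget props: plushie, gaming table, brick-stack game and camera.
# Visibility keys swap actors and props per shot, ending on a moving taped photo.
# Event tracks drive battle pass light switchers and the frontend camera, named as setting the nav camera for sequence completion.
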LevelSequence /MiscReferencersOfCosmetics/Animation/Game/MainPlayer/Menu/BR/S12_BP/Photographer/Photographer_Intro_Animation
bindings [
 SHOT_0100 {
  id=#1
  possessed_class=CineCameraActor
  tracks [
   {
    type=MovieScene3DTransformTrack
    sections [{
      type=MovieScene3DTransformSection
      channels [
       Location.X {type=MovieSceneScriptingDoubleChannel, default=2079.5173, const [2450.3538,2]}
       Location.Y {type=MovieSceneScriptingDoubleChannel, default=-106897.5078, const [-106897.5078,2]}
       Location.Z {type=MovieSceneScriptingDoubleChannel, default=198.2322, const [198.2322,2]}
       Rotation.X {type=MovieSceneScriptingDoubleChannel, default=0.0001, const [0.0001,2]}
       Rotation.Y {type=MovieSceneScriptingDoubleChannel, default=0.3633, const [0.3633,2]}
       Rotation.Z {type=MovieSceneScriptingDoubleChannel, default=-89.221, const [-89.221,2]}
       Scale.X {type=MovieSceneScriptingDoubleChannel, default=1, const [1,2]}
       Scale.Y {type=MovieSceneScriptingDoubleChannel, default=1, const [1,2]}
       Scale.Z {type=MovieSceneScriptingDoubleChannel, default=1, const [1,2]}
      ]
     }]
   }
   {type=MovieSceneSpawnTrack, sections [{type=MovieSceneSpawnSection, channels [None {type=MovieSceneScriptingBoolChannel, default=true}]}]}
  ]
 }
 SHOT_0045 {
  id=#2
  possessed_class=CineCameraActor
  tracks [
   {
    type=MovieScene3DTransformTrack
    sections [{
      type=MovieScene3DTransformSection
      channels [
       Location.X {type=MovieSceneScriptingDoubleChannel, default=2155.0449}
       Location.Y {type=MovieSceneScriptingDoubleChannel, default=-106887.0156}
       Location.Z {type=MovieSceneScriptingDoubleChannel, default=118.2029}
       Rotation.X {type=MovieSceneScriptingDoubleChannel, default=0.0001}
       Rotation.Y {type=MovieSceneScriptingDoubleChannel, default=-3.0265}
       Rotation.Z {type=MovieSceneScriptingDoubleChannel, default=-234.8496}
       Scale.X {type=MovieSceneScriptingDoubleChannel, default=1}
       Scale.Y {type=MovieSceneScriptingDoubleChannel, default=1}
       Scale.Z {type=MovieSceneScriptingDoubleChannel, default=1}
      ]
     }]
   }
   {type=MovieSceneSpawnTrack, sections [{type=MovieSceneSpawnSection, channels [None {type=MovieSceneScriptingBoolChannel, default=true}]}]}
  ]
 }
 Photographer_Plushie_Gadget_Photographer_Plushie_LOD0 {
  id=#3
  possessed_class=SkeletalMeshActor
  tracks [
   {
    type=MovieScene3DTransformTrack
    sections [{
      type=MovieScene3DTransformSection
      channels [
       Location.X {type=MovieSceneScriptingDoubleChannel, default=2500, keys [[0,2488.7451],[72,2500],[402.0013,2488.7451]]}
       Location.Y {type=MovieSceneScriptingDoubleChannel, default=-107000, thinned_from=3, keys [[0,-107017.4062],[402.0013,-107017.4062]]}
       Location.Z {type=MovieSceneScriptingDoubleChannel, default=12, keys [[0,11],[72,23.576],[402.0013,11]]}
       Rotation.X {type=MovieSceneScriptingDoubleChannel, default=0, const [0,3]}
       Rotation.Y {type=MovieSceneScriptingDoubleChannel, default=0, const [0,3]}
       Rotation.Z {type=MovieSceneScriptingDoubleChannel, default=0, const [0,3]}
       Scale.X {type=MovieSceneScriptingDoubleChannel, default=1, const [1,3]}
       Scale.Y {type=MovieSceneScriptingDoubleChannel, default=1, const [1,3]}
       Scale.Z {type=MovieSceneScriptingDoubleChannel, default=1, const [1,3]}
      ]
     }]
   }
   {
    type=MovieSceneSkeletalAnimationTrack
    sections [
     {type=MovieSceneSkeletalAnimationSection, end_frame=222, start_frame=72}
     {type=MovieSceneSkeletalAnimationSection, end_frame=474, start_frame=402}
     {type=MovieSceneSkeletalAnimationSection, end_frame=72, start_frame=0}
    ]
   }
   {
    type=MovieSceneVisibilityTrack
    sections [{type=MovieSceneVisibilitySection, channels [None {type=MovieSceneScriptingBoolChannel, default=true, const [true,1]}]}]
   }
  ]
 }
 ConstructorMannequin_Male {
  id=#4
  possessed_class=ConstructorMannequin_Male_C
  tracks [
   {
    type=MovieScene3DTransformTrack
    sections [{
      type=MovieScene3DTransformSection
      channels [
       Location.X {type=MovieSceneScriptingDoubleChannel, default=2500}
       Location.Y {type=MovieSceneScriptingDoubleChannel, default=-107000}
       Location.Z {type=MovieSceneScriptingDoubleChannel, default=12}
       Rotation.X {type=MovieSceneScriptingDoubleChannel, default=0}
       Rotation.Y {type=MovieSceneScriptingDoubleChannel, default=0}
       Rotation.Z {type=MovieSceneScriptingDoubleChannel, default=0}
       Scale.X {type=MovieSceneScriptingDoubleChannel, default=1}
       Scale.Y {type=MovieSceneScriptingDoubleChannel, default=1}
       Scale.Z {type=MovieSceneScriptingDoubleChannel, default=1}
      ]
     }]
   }
   {
    type=MovieSceneSkeletalAnimationTrack
    sections [
     {type=MovieSceneSkeletalAnimationSection, end_frame=72, start_frame=37}
     {type=MovieSceneSkeletalAnimationSection, end_frame=37, start_frame=-1}
     {type=MovieSceneSkeletalAnimationSection, end_frame=295, start_frame=225}
     {type=MovieSceneSkeletalAnimationSection, end_frame=402, start_frame=332}
     {type=MovieSceneSkeletalAnimationSection, end_frame=452, start_frame=402}
     {type=MovieSceneSkeletalAnimationSection, end_frame=523, start_frame=489}
     {type=MovieSceneSkeletalAnimationSection, end_frame=461, start_frame=452}
     {type=MovieSceneSkeletalAnimationSection, end_frame=479, start_frame=462}
    ]
   }
   {
    type=MovieSceneVisibilityTrack
    sections [{
      type=MovieSceneVisibilitySection
      channels [None {type=MovieSceneScriptingBoolChannel, default=true, keys [[0,true],[295,false],[332,true]]}]
     }]
   }
  ]
 }
 CameraComponent {
  id=#5
  parent_id=#6
  possessed_class=CineCameraComponent
  tracks [
   {
    type=MovieSceneFloatTrack
    display_name=Current Focal Length
    sections [{type=MovieSceneFloatSection, channels [None {type=MovieSceneScriptingFloatChannel, default=33.3105}]}]
   }
   {
    type=MovieSceneFloatTrack
    display_name="Manual Focus Distance (Focus Settings)"
    sections [{type=MovieSceneFloatSection, channels [None {type=MovieSceneScriptingFloatChannel, default=254.0096}]}]
   }
   {
    type=MovieSceneFloatTrack
    display_name=Current Aperture
    sections [{type=MovieSceneFloatSection, channels [None {type=MovieSceneScriptingFloatChannel, default=0.3673}]}]
   }
  ]
 }
 CameraComponent {
  id=#7
  parent_id=#8
  possessed_class=CineCameraComponent
  tracks [
   {
    type=MovieSceneFloatTrack
    display_name=Current Focal Length
    sections [{type=MovieSceneFloatSection, channels [None {type=MovieSceneScriptingFloatChannel, default=32.5313}]}]
   }
   {
    type=MovieSceneFloatTrack
    display_name="Manual Focus Distance (Focus Settings)"
    sections [{type=MovieSceneFloatSection, channels [None {type=MovieSceneScriptingFloatChannel, default=188.3825}]}]
   }
   {
    type=MovieSceneFloatTrack
    display_name=Current Aperture
    sections [{type=MovieSceneFloatSection, channels [None {type=MovieSceneScriptingFloatChannel, default=62.8298}]}]
   }
  ]
 }
 CameraComponent {
  id=#9
  parent_id=#10
  possessed_class=CineCameraComponent
  tracks [
   {
    type=MovieSceneFloatTrack
    display_name=Current Focal Length
    sections [{type=MovieSceneFloatSection, channels [None {type=MovieSceneScriptingFloatChannel, default=28}]}]
   }
   {
    type=MovieSceneFloatTrack
    display_name="Manual Focus Distance (Focus Settings)"
    sections [{type=MovieSceneFloatSection, channels [None {type=MovieSceneScriptingFloatChannel, default=29.3346}]}]
   }
   {
    type=MovieSceneFloatTrack
    display_name=Current Aperture
    sections [{type=MovieSceneFloatSection, channels [None {type=MovieSceneScriptingFloatChannel, default=14.394}]}]
   }
  ]
 }
 CameraComponent {
  id=#11
  parent_id=#12
  possessed_class=CineCameraComponent
  tracks [
   {
    type=MovieSceneFloatTrack
    display_name=Current Focal Length
    sections [{type=MovieSceneFloatSection, channels [None {type=MovieSceneScriptingFloatChannel, default=35}]}]
   }
   {
    type=MovieSceneFloatTrack
    display_name="Manual Focus Distance (Focus Settings)"
    sections [{type=MovieSceneFloatSection, channels [None {type=MovieSceneScriptingFloatChannel, default=40.3956}]}]
   }
   {
    type=MovieSceneFloatTrack
    display_name=Current Aperture
    sections [{type=MovieSceneFloatSection, channels [None {type=MovieSceneScriptingFloatChannel, default=4.3988}]}]
   }
  ]
 }
 Builderblocks {
  id=#13
  possessed_class=FortStaticMeshActor
  tracks [
   {
    type=MovieScene3DTransformTrack
    sections [{
      type=MovieScene3DTransformSection
      channels [
       Location.X {type=MovieSceneScriptingDoubleChannel, default=2093.6768, keys [[0,2073.686],[212,2096.4421],[393.9987,2084.4421]]}
       Location.Y {type=MovieSceneScriptingDoubleChannel, default=-106812, const [-106811.375,2]}
       Location.Z {type=MovieSceneScriptingDoubleChannel, default=76, keys [[0,78.8121],[212,70.8121],[393.9987,70.8121]]}
       Rotation.X {type=MovieSceneScriptingDoubleChannel, default=0, const [0,3]}
       Rotation.Y {type=MovieSceneScriptingDoubleChannel, default=0, const [0,3]}
       Rotation.Z {type=MovieSceneScriptingDoubleChannel, default=34.5654, const [34.5654,3]}
       Scale.X {type=MovieSceneScriptingDoubleChannel, default=1, const [1,3]}
       Scale.Y {type=MovieSceneScriptingDoubleChannel, default=1, const [1,3]}
       Scale.Z {type=MovieSceneScriptingDoubleChannel, default=1, const [1,3]}
      ]
     }]
   }
   {
    type=MovieSceneSpawnTrack
    sections [{type=MovieSceneSpawnSection, channels [None {type=MovieSceneScriptingBoolChannel, default=true, const [true,1]}]}]
   }
   {
    type=MovieSceneVisibilityTrack
    sections [{type=MovieSceneVisibilitySection, channels [None {type=MovieSceneScriptingBoolChannel, default=false, const [false,7]}]}]
   }
  ]
 }
 CameraComponent {
  id=#14
  parent_id=#1
  possessed_class=CineCameraComponent
  tracks [
   {
    type=MovieSceneFloatTrack
    display_name=Current Focal Length
    sections [{type=MovieSceneFloatSection, channels [None {type=MovieSceneScriptingFloatChannel, default=50}]}]
   }
   {
    type=MovieSceneFloatTrack
    display_name="Manual Focus Distance (Focus Settings)"
    sections [{type=MovieSceneFloatSection, channels [None {type=MovieSceneScriptingFloatChannel, default=53.3288}]}]
   }
   {
    type=MovieSceneFloatTrack
    display_name=Current Aperture
    sections [{type=MovieSceneFloatSection, channels [None {type=MovieSceneScriptingFloatChannel, default=50.1591}]}]
   }
  ]
 }
 SHOT_0030 {
  id=#15
  possessed_class=CineCameraActor
  tracks [
   {
    type=MovieScene3DTransformTrack
    sections [{
      type=MovieScene3DTransformSection
      channels [
       Location.X {type=MovieSceneScriptingDoubleChannel, default=2157.2815, const [2157.2815,4]}
       Location.Y {type=MovieSceneScriptingDoubleChannel, default=-106842.4688, const [-106842.4688,4]}
       Location.Z {type=MovieSceneScriptingDoubleChannel, default=100.4411, const [103.882,4]}
       Rotation.X {type=MovieSceneScriptingDoubleChannel, default=0, const [0,4]}
       Rotation.Y {type=MovieSceneScriptingDoubleChannel, default=1.1748, keys [[110,1.1748],[137,0.8064],[162,0.6641],[193,2.5318]]}
       Rotation.Z {type=MovieSceneScriptingDoubleChannel, default=-217.2368, keys [[110,-215.077],[137,-220.1869],[162,-220.7057],[193,-220.5307]]}
       Scale.X {type=MovieSceneScriptingDoubleChannel, default=1, const [1,4]}
       Scale.Y {type=MovieSceneScriptingDoubleChannel, default=1, const [1,4]}
       Scale.Z {type=MovieSceneScriptingDoubleChannel, default=1, const [1,4]}
      ]
     }]
   }
   {
    type=MovieSceneSpawnTrack
    sections [{type=MovieSceneSpawnSection, channels [None {type=MovieSceneScriptingBoolChannel, default=true, const [true,1]}]}]
   }
  ]
 }
 CameraComponent {
  id=#16
  parent_id=#15
  possessed_class=CineCameraComponent
  tracks [
   {
    type=MovieSceneFloatTrack
    display_name=Current Focal Length
    sections [{type=MovieSceneFloatSection, channels [None {type=MovieSceneScriptingFloatChannel, default=28}]}]
   }
   {
    type=MovieSceneFloatTrack
    display_name="Manual Focus Distance (Focus Settings)"
    sections [{type=MovieSceneFloatSection, channels [None {type=MovieSceneScriptingFloatChannel, default=90.2451}]}]
   }
   {
    type=MovieSceneFloatTrack
    display_name=Current Aperture
    sections [{type=MovieSceneFloatSection, channels [None {type=MovieSceneScriptingFloatChannel, default=6.0115}]}]
   }
  ]
 }
 SoldierMannequin_Female {
  id=#17
  possessed_class=SoldierMannequin_Male_C
  tracks [
   {
    type=MovieScene3DTransformTrack
    sections [{
      type=MovieScene3DTransformSection
      channels [
       Location.X {type=MovieSceneScriptingDoubleChannel, default=2500}
       Location.Y {type=MovieSceneScriptingDoubleChannel, default=-107000}
       Location.Z {type=MovieSceneScriptingDoubleChannel, default=12}
       Rotation.X {type=MovieSceneScriptingDoubleChannel, default=0}
       Rotation.Y {type=MovieSceneScriptingDoubleChannel, default=0}
       Rotation.Z {type=MovieSceneScriptingDoubleChannel, default=0}
       Scale.X {type=MovieSceneScriptingDoubleChannel, default=1}
       Scale.Y {type=MovieSceneScriptingDoubleChannel, default=1}
       Scale.Z {type=MovieSceneScriptingDoubleChannel, default=1}
      ]
     }]
   }
   {
    type=MovieSceneSkeletalAnimationTrack
    sections [
     {type=MovieSceneSkeletalAnimationSection, end_frame=258, start_frame=225}
     {type=MovieSceneSkeletalAnimationSection, end_frame=225, start_frame=72}
     {type=MovieSceneSkeletalAnimationSection, end_frame=407, start_frame=357}
     {type=MovieSceneSkeletalAnimationSection, end_frame=456, start_frame=406}
     {type=MovieSceneSkeletalAnimationSection, end_frame=37, start_frame=0}
     {type=MovieSceneSkeletalAnimationSection, end_frame=295, start_frame=258}
     {type=MovieSceneSkeletalAnimationSection, end_frame=494, start_frame=479}
     {type=MovieSceneSkeletalAnimationSection, end_frame=72, start_frame=37}
     {type=MovieSceneSkeletalAnimationSection, end_frame=332, start_frame=295}
    ]
   }
   {
    type=MovieSceneVisibilityTrack
    sections [{type=MovieSceneVisibilitySection, channels [None {type=MovieSceneScriptingBoolChannel, default=true, const [true,1]}]}]
   }
  ]
 }
 Photographer__Gaming_Table_01_Gadget {
  id=#18
  possessed_class=SkeletalMeshActor
  tracks [
   {
    type=MovieScene3DTransformTrack
    sections [{
      type=MovieScene3DTransformSection
      channels [
       Location.X {type=MovieSceneScriptingDoubleChannel, default=2500, const [2500,1]}
       Location.Y {type=MovieSceneScriptingDoubleChannel, default=-107000, const [-107000,1]}
       Location.Z {type=MovieSceneScriptingDoubleChannel, default=12, const [12,1]}
       Rotation.X {type=MovieSceneScriptingDoubleChannel, default=0, const [0,1]}
       Rotation.Y {type=MovieSceneScriptingDoubleChannel, default=0, const [0,1]}
       Rotation.Z {type=MovieSceneScriptingDoubleChannel, default=0, const [0,1]}
       Scale.X {type=MovieSceneScriptingDoubleChannel, default=1, const [1,1]}
       Scale.Y {type=MovieSceneScriptingDoubleChannel, default=1, const [1,1]}
       Scale.Z {type=MovieSceneScriptingDoubleChannel, default=1, const [1,1]}
      ]
     }]
   }
   {
    type=MovieSceneSkeletalAnimationTrack
    sections [
     {type=MovieSceneSkeletalAnimationSection, end_frame=462, start_frame=452}
     {type=MovieSceneSkeletalAnimationSection, end_frame=479, start_frame=462}
     {type=MovieSceneSkeletalAnimationSection, end_frame=489, start_frame=479}
     {type=MovieSceneSkeletalAnimationSection, end_frame=452, start_frame=0}
    ]
   }
   {
    type=MovieSceneVisibilityTrack
    sections [{type=MovieSceneVisibilitySection, channels [None {type=MovieSceneScriptingBoolChannel, default=true, keys [[0,true],[462,false]]}]}]
   }
  ]
 }
 BP_Battlepass_LightSwitcher_Maproom {
  id=#19
  possessed_class=BP_Battlepass_LightSwitcher_C
  tracks [{
    type=MovieSceneEventTrack
    sections [{
      type=MovieSceneEventTriggerSection
      channels [None {type=MovieSceneScriptingEventChannel, const ["<Struct 'MovieSceneEvent' (0x8518df5a640) {}>", 1]}]
     }]
   }]
 }
 Photo Intro 0080 Photographer Locker Item Gadget Photographer Camera {
  id=#20
  possessed_class=SkeletalMeshActor
  tracks [
   {
    type=MovieScene3DTransformTrack
    sections [{
      type=MovieScene3DTransformSection
      channels [
       Location.X {type=MovieSceneScriptingDoubleChannel, default=2500}
       Location.Y {type=MovieSceneScriptingDoubleChannel, default=-107000}
       Location.Z {type=MovieSceneScriptingDoubleChannel, default=12}
       Rotation.X {type=MovieSceneScriptingDoubleChannel, default=0}
       Rotation.Y {type=MovieSceneScriptingDoubleChannel, default=0}
       Rotation.Z {type=MovieSceneScriptingDoubleChannel, default=90}
       Scale.X {type=MovieSceneScriptingDoubleChannel, default=1}
       Scale.Y {type=MovieSceneScriptingDoubleChannel, default=1}
       Scale.Z {type=MovieSceneScriptingDoubleChannel, default=1}
      ]
     }]
   }
   {type=MovieSceneSkeletalAnimationTrack, sections [{type=MovieSceneSkeletalAnimationSection, end_frame=494, start_frame=479}]}
   {type=MovieSceneSpawnTrack, sections [{type=MovieSceneSpawnSection, channels [None {type=MovieSceneScriptingBoolChannel, default=true}]}]}
   {
    type=MovieSceneVisibilityTrack
    sections [{
      type=MovieSceneVisibilitySection
      channels [None {type=MovieSceneScriptingBoolChannel, default=false, keys [[0,false],[479,true],[489,false]]}]
     }]
   }
  ]
 }
 BP_Frontend_EventLevel_NavObject_Mannequin3 {
  id=#21
  possessed_class=BP_Frontend_EventLevel_NavObject_Mannequin_C
  tracks [{
    type=MovieSceneVisibilityTrack
    sections [{type=MovieSceneVisibilitySection, channels [None {type=MovieSceneScriptingBoolChannel, default=false}]}]
   }]
 }
 Photographer_TapedPhoto_Gadget {
  id=#22
  possessed_class=SkeletalMeshActor
  tracks [
   {
    type=MovieScene3DTransformTrack
    sections [{
      type=MovieScene3DTransformSection
      channels [
       Location.X {type=MovieSceneScriptingDoubleChannel, default=2078.4534, keys [[508.75,2450.7031],[547.5,2449.2898]]}
       Location.Y {type=MovieSceneScriptingDoubleChannel, default=-106944.3359, keys [[509,-106936.8906],[548,-106980.8281]]}
       Location.Z {type=MovieSceneScriptingDoubleChannel, default=198.74, keys [[509,196.4811],[548,198.74]]}
       Rotation.X {type=MovieSceneScriptingDoubleChannel, default=93.2852, keys [[509,92.4673],[548,43.8676]]}
       Rotation.Y {type=MovieSceneScriptingDoubleChannel, default=7.817, keys [[509,-0.1253],[548,7.817]]}
       Rotation.Z {type=MovieSceneScriptingDoubleChannel, default=75.9552, keys [[509,0.7481],[548,75.9552]]}
       Scale.X {type=MovieSceneScriptingDoubleChannel, default=1.135, keys [[509,1.135],[548,1.5984]]}
       Scale.Y {type=MovieSceneScriptingDoubleChannel, default=0.7232, keys [[509,0.8973],[548,0.8973]]}
       Scale.Z {type=MovieSceneScriptingDoubleChannel, default=1, const [1,2]}
      ]
     }]
   }
   {type=MovieSceneSkeletalAnimationTrack, sections []}
   {
    type=MovieSceneVisibilityTrack
    sections [{
      type=MovieSceneVisibilitySection
      channels [None {type=MovieSceneScriptingBoolChannel, default=false, keys [[0,false],[509,true],[549,false]]}]
     }]
   }
  ]
 }
 CameraComponent {
  id=#23
  parent_id=#24
  possessed_class=CineCameraComponent
  tracks [
   {
    type=MovieSceneFloatTrack
    display_name=Current Focal Length
    sections [{type=MovieSceneFloatSection, channels [None {type=MovieSceneScriptingFloatChannel, default=35}]}]
   }
   {
    type=MovieSceneFloatTrack
    display_name="Manual Focus Distance (Focus Settings)"
    sections [{type=MovieSceneFloatSection, channels [None {type=MovieSceneScriptingFloatChannel, default=47.549}]}]
   }
   {
    type=MovieSceneFloatTrack
    display_name=Current Aperture
    sections [{type=MovieSceneFloatSection, channels [None {type=MovieSceneScriptingFloatChannel, default=8.3875}]}]
   }
  ]
 }
 Photographer_BrickStackGame_Gadget {
  id=#25
  possessed_class=SkeletalMeshActor
  tracks [
   {
    type=MovieScene3DTransformTrack
    sections [{
      type=MovieScene3DTransformSection
      channels [
       Location.X {type=MovieSceneScriptingDoubleChannel, default=2500}
       Location.Y {type=MovieSceneScriptingDoubleChannel, default=-107000}
       Location.Z {type=MovieSceneScriptingDoubleChannel, default=12}
       Rotation.X {type=MovieSceneScriptingDoubleChannel, default=0}
       Rotation.Y {type=MovieSceneScriptingDoubleChannel, default=0}
       Rotation.Z {type=MovieSceneScriptingDoubleChannel, default=90}
       Scale.X {type=MovieSceneScriptingDoubleChannel, default=1}
       Scale.Y {type=MovieSceneScriptingDoubleChannel, default=1}
       Scale.Z {type=MovieSceneScriptingDoubleChannel, default=1}
      ]
     }]
   }
   {
    type=MovieSceneSkeletalAnimationTrack
    sections [
     {type=MovieSceneSkeletalAnimationSection, end_frame=72, start_frame=37}
     {type=MovieSceneSkeletalAnimationSection, end_frame=37, start_frame=0}
     {type=MovieSceneSkeletalAnimationSection, end_frame=225, start_frame=72}
     {type=MovieSceneSkeletalAnimationSection, end_frame=402, start_frame=332}
     {type=MovieSceneSkeletalAnimationSection, end_frame=295, start_frame=225}
    ]
   }
   {
    type=MovieSceneVisibilityTrack
    sections [{
      type=MovieSceneVisibilitySection
      channels [None {type=MovieSceneScriptingBoolChannel, default=true, keys [[0,true],[295,false],[332,true]]}]
     }]
   }
   {
    type=MovieSceneSkeletalAnimationTrack
    sections [
     {type=MovieSceneSkeletalAnimationSection, end_frame=462, start_frame=452}
     {type=MovieSceneSkeletalAnimationSection, end_frame=479, start_frame=462}
     {type=MovieSceneSkeletalAnimationSection, end_frame=529, start_frame=489}
     {type=MovieSceneSkeletalAnimationSection, end_frame=452, start_frame=402}
     {type=MovieSceneSkeletalAnimationSection, end_frame=489, start_frame=479}
    ]
   }
  ]
 }
 CameraComponent {
  id=#26
  parent_id=#27
  possessed_class=CineCameraComponent
  tracks [
   {
    type=MovieSceneFloatTrack
    display_name=Current Focal Length
    sections [{type=MovieSceneFloatSection, channels [None {type=MovieSceneScriptingFloatChannel, default=35, const [35,1]}]}]
   }
   {
    type=MovieSceneFloatTrack
    display_name="Manual Focus Distance (Focus Settings)"
    sections [{type=MovieSceneFloatSection, channels [None {type=MovieSceneScriptingFloatChannel, default=110.5326, const [100000,1]}]}]
   }
   {
    type=MovieSceneFloatTrack
    display_name=Current Aperture
    sections [{type=MovieSceneFloatSection, channels [None {type=MovieSceneScriptingFloatChannel, default=5.8, const [5.8,1]}]}]
   }
  ]
 }
 CameraComponent {
  id=#28
  parent_id=#2
  possessed_class=CineCameraComponent
  tracks [
   {
    type=MovieSceneFloatTrack
    display_name=Current Focal Length
    sections [{type=MovieSceneFloatSection, channels [None {type=MovieSceneScriptingFloatChannel, default=35}]}]
   }
   {
    type=MovieSceneFloatTrack
    display_name="Manual Focus Distance (Focus Settings)"
    sections [{type=MovieSceneFloatSection, channels [None {type=MovieSceneScriptingFloatChannel, default=158.1285}]}]
   }
   {
    type=MovieSceneFloatTrack
    display_name=Current Aperture
    sections [{type=MovieSceneFloatSection, channels [None {type=MovieSceneScriptingFloatChannel, default=4.4909}]}]
   }
  ]
 }
 BP_Battlepass_LightSwitcher_Adventure {
  id=#29
  possessed_class=BP_Battlepass_LightSwitcher_C
  tracks [{
    type=MovieSceneEventTrack
    sections [{
      type=MovieSceneEventTriggerSection
      channels [None {type=MovieSceneScriptingEventChannel, const ["<Struct 'MovieSceneEvent' (0x8518df58de0) {}>", 1]}]
     }]
   }]
 }
 SHOT_0010 {
  id=#10
  possessed_class=CineCameraActor
  tracks [
   {
    type=MovieScene3DTransformTrack
    sections [{
      type=MovieScene3DTransformSection
      channels [
       Location.X {type=MovieSceneScriptingDoubleChannel, default=2098.3708}
       Location.Y {type=MovieSceneScriptingDoubleChannel, default=-106786.1953}
       Location.Z {type=MovieSceneScriptingDoubleChannel, default=142.3092}
       Rotation.X {type=MovieSceneScriptingDoubleChannel, default=0.0001}
       Rotation.Y {type=MovieSceneScriptingDoubleChannel, default=-3.8248}
       Rotation.Z {type=MovieSceneScriptingDoubleChannel, default=241.5745}
       Scale.X {type=MovieSceneScriptingDoubleChannel, default=1}
       Scale.Y {type=MovieSceneScriptingDoubleChannel, default=1}
       Scale.Z {type=MovieSceneScriptingDoubleChannel, default=1}
      ]
     }]
   }
   {type=MovieSceneSpawnTrack, sections [{type=MovieSceneSpawnSection, channels [None {type=MovieSceneScriptingBoolChannel, default=true}]}]}
  ]
 }
 CameraComponent {
  id=#30
  parent_id=#31
  possessed_class=CineCameraComponent
  tracks [
   {
    type=MovieSceneFloatTrack
    display_name=Current Focal Length
    sections [{type=MovieSceneFloatSection, channels [None {type=MovieSceneScriptingFloatChannel, default=60}]}]
   }
   {
    type=MovieSceneFloatTrack
    display_name="Manual Focus Distance (Focus Settings)"
    sections [{type=MovieSceneFloatSection, channels [None {type=MovieSceneScriptingFloatChannel, default=154.314}]}]
   }
   {
    type=MovieSceneFloatTrack
    display_name=Current Aperture
    sections [{type=MovieSceneFloatSection, channels [None {type=MovieSceneScriptingFloatChannel, default=6.182}]}]
   }
  ]
 }
 CameraComponent {
  id=#32
  parent_id=#33
  possessed_class=CineCameraComponent
  tracks [
   {
    type=MovieSceneFloatTrack
    display_name=Current Focal Length
    sections [{type=MovieSceneFloatSection, channels [None {type=MovieSceneScriptingFloatChannel, default=35}]}]
   }
   {
    type=MovieSceneFloatTrack
    display_name="Manual Focus Distance (Focus Settings)"
    sections [{type=MovieSceneFloatSection, channels [None {type=MovieSceneScriptingFloatChannel, default=260.6498}]}]
   }
   {
    type=MovieSceneFloatTrack
    display_name=Current Aperture
    sections [{type=MovieSceneFloatSection, channels [None {type=MovieSceneScriptingFloatChannel, default=1.8166}]}]
   }
  ]
 }
 SHOT_0050 {
  id=#24
  possessed_class=CineCameraActor
  tracks [
   {
    type=MovieScene3DTransformTrack
    sections [{
      type=MovieScene3DTransformSection
      channels [
       Location.X {type=MovieSceneScriptingDoubleChannel, default=2130.6301}
       Location.Y {type=MovieSceneScriptingDoubleChannel, default=-106787.625}
       Location.Z {type=MovieSceneScriptingDoubleChannel, default=122.702}
       Rotation.X {type=MovieSceneScriptingDoubleChannel, default=0}
       Rotation.Y {type=MovieSceneScriptingDoubleChannel, default=-0.0275}
       Rotation.Z {type=MovieSceneScriptingDoubleChannel, default=-147.6367}
       Scale.X {type=MovieSceneScriptingDoubleChannel, default=1}
       Scale.Y {type=MovieSceneScriptingDoubleChannel, default=1}
       Scale.Z {type=MovieSceneScriptingDoubleChannel, default=1}
      ]
     }]
   }
   {type=MovieSceneSpawnTrack, sections [{type=MovieSceneSpawnSection, channels [None {type=MovieSceneScriptingBoolChannel, default=true}]}]}
  ]
 }
 SHOT_0060 {
  id=#33
  possessed_class=CineCameraActor
  tracks [
   {
    type=MovieScene3DTransformTrack
    sections [{
      type=MovieScene3DTransformSection
      channels [
       Location.X {type=MovieSceneScriptingDoubleChannel, default=2295.4666}
       Location.Y {type=MovieSceneScriptingDoubleChannel, default=-106658.8359}
       Location.Z {type=MovieSceneScriptingDoubleChannel, default=79.5108}
       Rotation.X {type=MovieSceneScriptingDoubleChannel, default=0}
       Rotation.Y {type=MovieSceneScriptingDoubleChannel, default=7.5724}
       Rotation.Z {type=MovieSceneScriptingDoubleChannel, default=-145.4321}
       Scale.X {type=MovieSceneScriptingDoubleChannel, default=1}
       Scale.Y {type=MovieSceneScriptingDoubleChannel, default=1}
       Scale.Z {type=MovieSceneScriptingDoubleChannel, default=1}
      ]
     }]
   }
   {type=MovieSceneSpawnTrack, sections [{type=MovieSceneSpawnSection, channels [None {type=MovieSceneScriptingBoolChannel, default=true}]}]}
  ]
 }
 SHOT_0020 {
  id=#6
  possessed_class=CineCameraActor
  tracks [
   {
    type=MovieScene3DTransformTrack
    sections [{
      type=MovieScene3DTransformSection
      channels [
       Location.X {type=MovieSceneScriptingDoubleChannel, default=2331.0186, keys [[37,2299.5947],[72,2296.938]]}
       Location.Y {type=MovieSceneScriptingDoubleChannel, default=-106882.0625, keys [[37,-106671.0547],[72,-106666.4688]]}
       Location.Z {type=MovieSceneScriptingDoubleChannel, default=64.6398, const [80,2]}
       Rotation.X {type=MovieSceneScriptingDoubleChannel, default=0, const [0,2]}
       Rotation.Y {type=MovieSceneScriptingDoubleChannel, default=6.03, const [6,2]}
       Rotation.Z {type=MovieSceneScriptingDoubleChannel, default=171.971, keys [[37,214.7689],[72,214.769]]}
       Scale.X {type=MovieSceneScriptingDoubleChannel, default=1, const [1,2]}
       Scale.Y {type=MovieSceneScriptingDoubleChannel, default=1, const [1,2]}
       Scale.Z {type=MovieSceneScriptingDoubleChannel, default=1, const [1,2]}
      ]
     }]
   }
   {type=MovieSceneSpawnTrack, sections [{type=MovieSceneSpawnSection, channels [None {type=MovieSceneScriptingBoolChannel, default=true}]}]}
  ]
 }
 "Camera_Root_Movement (Setting Nav Camera for once Sequence Completes)" {
  id=#34
  possessed_class=Frontend_EventLevel_Camera_C
  tracks [
   {
    type=MovieScene3DTransformTrack
    sections [{
      type=MovieScene3DTransformSection
      channels [
       Location.X {type=MovieSceneScriptingDoubleChannel, default=2527.9539, keys [[0,2911.8142],[540,2527.9539]]}
       Location.Y {type=MovieSceneScriptingDoubleChannel, default=-106702.7891, keys [[0,-106705.4922],[540,-106702.7891]]}
       Location.Z {type=MovieSceneScriptingDoubleChannel, default=104.2625, keys [[0,159],[540,104.2625]]}
       Rotation.X {type=MovieSceneScriptingDoubleChannel, default=0, keys [[0,0],[540,0]]}
       Rotation.Y {type=MovieSceneScriptingDoubleChannel, default=1.7322, keys [[0,0],[540,1.7322]]}
       Rotation.Z {type=MovieSceneScriptingDoubleChannel, default=-149.2204, keys [[0,-156.2538],[540,-149.2204]]}
       Scale.X {type=MovieSceneScriptingDoubleChannel, default=1, const [1,1]}
       Scale.Y {type=MovieSceneScriptingDoubleChannel, default=1, const [1,1]}
       Scale.Z {type=MovieSceneScriptingDoubleChannel, default=1, const [1,1]}
      ]
     }]
   }
   {
    type=MovieSceneEventTrack
    sections [{
      type=MovieSceneEventTriggerSection
      channels [None {type=MovieSceneScriptingEventChannel, const ["<Struct 'MovieSceneEvent' (0x8518df5b540) {}>", 1]}]
     }]
   }
  ]
 }
 SHOT_0040 {
  id=#35
  possessed_class=CineCameraActor
  tracks [
   {
    type=MovieScene3DTransformTrack
    sections [{
      type=MovieScene3DTransformSection
      channels [
       Location.X {type=MovieSceneScriptingDoubleChannel, default=2054.4146, const [2070.3997,1]}
       Location.Y {type=MovieSceneScriptingDoubleChannel, default=-106743.4297, const [-106768.2969,1]}
       Location.Z {type=MovieSceneScriptingDoubleChannel, default=104.6197}
       Rotation.X {type=MovieSceneScriptingDoubleChannel, default=0}
       Rotation.Y {type=MovieSceneScriptingDoubleChannel, default=5.3734}
       Rotation.Z {type=MovieSceneScriptingDoubleChannel, default=302.958}
       Scale.X {type=MovieSceneScriptingDoubleChannel, default=1}
       Scale.Y {type=MovieSceneScriptingDoubleChannel, default=1}
       Scale.Z {type=MovieSceneScriptingDoubleChannel, default=1}
      ]
     }]
   }
   {type=MovieSceneSpawnTrack, sections [{type=MovieSceneSpawnSection, channels [None {type=MovieSceneScriptingBoolChannel, default=true}]}]}
  ]
 }
 SHOT_0090 {
  id=#8
  possessed_class=CineCameraActor
  tracks [
   {
    type=MovieScene3DTransformTrack
    sections [{
      type=MovieScene3DTransformSection
      channels [
       Location.X {type=MovieSceneScriptingDoubleChannel, default=2044.6108, const [2044.6108,3]}
       Location.Y {type=MovieSceneScriptingDoubleChannel, default=-106727.0234, const [-106727.0234,3]}
       Location.Z {type=MovieSceneScriptingDoubleChannel, default=131.4507, const [131.4507,3]}
       Rotation.X {type=MovieSceneScriptingDoubleChannel, default=0.0001, const [0.0001,3]}
       Rotation.Y {type=MovieSceneScriptingDoubleChannel, default=10.0599, keys [[442,-13.6011],[475.0013,-13.6011],[486.0013,10.3648]]}
       Rotation.Z {type=MovieSceneScriptingDoubleChannel, default=306.765, const [306.765,3]}
       Scale.X {type=MovieSceneScriptingDoubleChannel, default=1, const [1,3]}
       Scale.Y {type=MovieSceneScriptingDoubleChannel, default=1, const [1,3]}
       Scale.Z {type=MovieSceneScriptingDoubleChannel, default=1, const [1,3]}
      ]
     }]
   }
   {type=MovieSceneSpawnTrack, sections [{type=MovieSceneSpawnSection, channels [None {type=MovieSceneScriptingBoolChannel, default=true}]}]}
  ]
 }
 SHOT_0065 {
  id=#36
  possessed_class=CineCameraActor
  tracks [
   {
    type=MovieScene3DTransformTrack
    sections [{
      type=MovieScene3DTransformSection
      channels [
       Location.X {type=MovieSceneScriptingDoubleChannel, default=2061.7971, const [2234.3994,1]}
       Location.Y {type=MovieSceneScriptingDoubleChannel, default=-106961.9375, const [-106807.1953,1]}
       Location.Z {type=MovieSceneScriptingDoubleChannel, default=73.3787, const [58.1445,1]}
       Rotation.X {type=MovieSceneScriptingDoubleChannel, default=0.0001, const [0.0001,1]}
       Rotation.Y {type=MovieSceneScriptingDoubleChannel, default=-0.8279, const [6.1713,1]}
       Rotation.Z {type=MovieSceneScriptingDoubleChannel, default=64.3665, const [199.972,1]}
       Scale.X {type=MovieSceneScriptingDoubleChannel, default=1, const [1,1]}
       Scale.Y {type=MovieSceneScriptingDoubleChannel, default=1, const [1,1]}
       Scale.Z {type=MovieSceneScriptingDoubleChannel, default=1, const [1,1]}
      ]
     }]
   }
   {type=MovieSceneSpawnTrack, sections [{type=MovieSceneSpawnSection, channels [None {type=MovieSceneScriptingBoolChannel, default=true}]}]}
  ]
 }
 SHOT_0035 {
  id=#12
  possessed_class=CineCameraActor
  tracks [
   {
    type=MovieScene3DTransformTrack
    sections [{
      type=MovieScene3DTransformSection
      channels [
       Location.X {type=MovieSceneScriptingDoubleChannel, default=2113.4424, const [2126.8145,4]}
       Location.Y {type=MovieSceneScriptingDoubleChannel, default=-106804.5781, const [-106812.2188,4]}
       Location.Z {type=MovieSceneScriptingDoubleChannel, default=125.4115, const [134.3867,4]}
       Rotation.X {type=MovieSceneScriptingDoubleChannel, default=0, const [0,4]}
       Rotation.Y {type=MovieSceneScriptingDoubleChannel, default=8.3735, keys [[193,1.1339],[207,7.0303],[231,8.3735],[300,8.3735]]}
       Rotation.Z {type=MovieSceneScriptingDoubleChannel, default=-202.8486, keys [[193,-188.5475],[207,-185.5272],[231,-183.9213],[300,-183.9213]]}
       Scale.X {type=MovieSceneScriptingDoubleChannel, default=1, const [1,4]}
       Scale.Y {type=MovieSceneScriptingDoubleChannel, default=1, const [1,4]}
       Scale.Z {type=MovieSceneScriptingDoubleChannel, default=1, const [1,4]}
      ]
     }]
   }
   {type=MovieSceneSpawnTrack, sections [{type=MovieSceneSpawnSection, channels [None {type=MovieSceneScriptingBoolChannel, default=true}]}]}
  ]
 }
 CameraComponent {
  id=#37
  parent_id=#35
  possessed_class=CineCameraComponent
  tracks [
   {
    type=MovieSceneFloatTrack
    display_name=Current Focal Length
    sections [{type=MovieSceneFloatSection, channels [None {type=MovieSceneScriptingFloatChannel, default=23.805}]}]
   }
   {
    type=MovieSceneFloatTrack
    display_name="Manual Focus Distance (Focus Settings)"
    sections [{type=MovieSceneFloatSection, channels [None {type=MovieSceneScriptingFloatChannel, default=56.5153}]}]
   }
   {
    type=MovieSceneFloatTrack
    display_name=Current Aperture
    sections [{type=MovieSceneFloatSection, channels [None {type=MovieSceneScriptingFloatChannel, default=5.819}]}]
   }
  ]
 }
 SkeletalMeshComponent0 {
  id=#38
  parent_id=#4
  possessed_class=SkeletalMeshComponent
  tracks [{
    type=MovieSceneFloatTrack
    display_name=Bounds Scale
    sections [{type=MovieSceneFloatSection, channels [None {type=MovieSceneScriptingFloatChannel, default=100}]}]
   }]
 }
 SHOT_0080 {
  id=#31
  possessed_class=CineCameraActor
  tracks [
   {
    type=MovieScene3DTransformTrack
    sections [{
      type=MovieScene3DTransformSection
      channels [
       Location.X {type=MovieSceneScriptingDoubleChannel, default=2211.3655, keys [[479.0025,2091.6646],[488.0025,2085.4165]]}
       Location.Y {type=MovieSceneScriptingDoubleChannel, default=-106953.1562, keys [[479.0013,-106877.7734],[488.0013,-106877.2344]]}
       Location.Z {type=MovieSceneScriptingDoubleChannel, default=170.9742, keys [[479.0025,180.6552],[488.0025,180.6549]]}
       Rotation.X {type=MovieSceneScriptingDoubleChannel, default=0.0001, keys [[479.0025,-2.4107],[488.0013,0.0001]]}
       Rotation.Y {type=MovieSceneScriptingDoubleChannel, default=-0.3404, keys [[479.0013,-22.5271],[488.0025,-22.3199]]}
       Rotation.Z {type=MovieSceneScriptingDoubleChannel, default=130.7651, keys [[479.0025,112.8577],[488.0025,110.5705]]}
       Scale.X {type=MovieSceneScriptingDoubleChannel, default=1, const [1,2]}
       Scale.Y {type=MovieSceneScriptingDoubleChannel, default=1, const [1,2]}
       Scale.Z {type=MovieSceneScriptingDoubleChannel, default=1, const [1,2]}
      ]
     }]
   }
   {type=MovieSceneSpawnTrack, sections [{type=MovieSceneSpawnSection, channels [None {type=MovieSceneScriptingBoolChannel, default=true}]}]}
  ]
 }
 CameraComponent {
  id=#39
  parent_id=#36
  possessed_class=CineCameraComponent
  tracks [
   {
    type=MovieSceneFloatTrack
    display_name=Current Focal Length
    sections [{type=MovieSceneFloatSection, channels [None {type=MovieSceneScriptingFloatChannel, default=35}]}]
   }
   {
    type=MovieSceneFloatTrack
    display_name="Manual Focus Distance (Focus Settings)"
    sections [{type=MovieSceneFloatSection, channels [None {type=MovieSceneScriptingFloatChannel, default=113.9362}]}]
   }
   {
    type=MovieSceneFloatTrack
    display_name=Current Aperture
    sections [{type=MovieSceneFloatSection, channels [None {type=MovieSceneScriptingFloatChannel, default=2.9649}]}]
   }
  ]
 }
 SHOT_0070 {
  id=#27
  possessed_class=CineCameraActor
  tracks [
   {
    type=MovieScene3DTransformTrack
    sections [{
      type=MovieScene3DTransformSection
      channels [
       Location.X {type=MovieSceneScriptingDoubleChannel, default=2061.7971, const [2211.3655,2]}
       Location.Y {type=MovieSceneScriptingDoubleChannel, default=-106961.9375, const [-106953.1562,2]}
       Location.Z {type=MovieSceneScriptingDoubleChannel, default=73.3787, const [170.9742,2]}
       Rotation.X {type=MovieSceneScriptingDoubleChannel, default=0.0001, const [0.0001,2]}
       Rotation.Y {type=MovieSceneScriptingDoubleChannel, default=-0.3404, keys [[463,-12.4281],[470.0013,-0.3404]]}
       Rotation.Z {type=MovieSceneScriptingDoubleChannel, default=64.3665, const [130.7652,2]}
       Scale.X {type=MovieSceneScriptingDoubleChannel, default=1, const [1,2]}
       Scale.Y {type=MovieSceneScriptingDoubleChannel, default=1, const [1,2]}
       Scale.Z {type=MovieSceneScriptingDoubleChannel, default=1, const [1,2]}
      ]
     }]
   }
   {
    type=MovieSceneSpawnTrack
    sections [{type=MovieSceneSpawnSection, channels [None {type=MovieSceneScriptingBoolChannel, default=true, const [true,1]}]}]
   }
  ]
 }
 S_Shark_Base_GamingTable_2 {
  id=#40
  possessed_class=FortStaticMeshActor
  tracks [
   {
    type=MovieScene3DTransformTrack
    sections [{
      type=MovieScene3DTransformSection
      channels [
       Location.X {type=MovieSceneScriptingDoubleChannel, default=2083.6582}
       Location.Y {type=MovieSceneScriptingDoubleChannel, default=-106804.1094}
       Location.Z {type=MovieSceneScriptingDoubleChannel, default=13.6298}
       Rotation.X {type=MovieSceneScriptingDoubleChannel, default=0}
       Rotation.Y {type=MovieSceneScriptingDoubleChannel, default=0}
       Rotation.Z {type=MovieSceneScriptingDoubleChannel, default=-55.2261}
       Scale.X {type=MovieSceneScriptingDoubleChannel, default=0.6307}
       Scale.Y {type=MovieSceneScriptingDoubleChannel, default=0.7845}
       Scale.Z {type=MovieSceneScriptingDoubleChannel, default=1.1901}
      ]
     }]
   }
   {
    type=MovieSceneVisibilityTrack
    sections [{type=MovieSceneVisibilitySection, channels [None {type=MovieSceneScriptingBoolChannel, default=false}]}]
   }
  ]
 }
 CameraComponent {
  id=#41
  parent_id=#34
  possessed_class=CameraComponent
  tracks [{
    type=MovieSceneFloatTrack
    display_name=Field Of View
    sections [{type=MovieSceneFloatSection, channels [None {type=MovieSceneScriptingFloatChannel, default=85, keys [[0,80],[540,85]]}]}]
   }]
 }
]
display_rate {denominator=1, numerator=30}
tick_resolution {denominator=1, numerator=24000}
tracks [
 {
  type=MovieSceneAudioTrack
  display_name=Audio
  sections [{
    type=MovieSceneAudioSection
    end_frame=547
    start_frame=0
    channels [
     Volume {type=MovieSceneScriptingFloatChannel, default=1}
     Pitch {type=MovieSceneScriptingFloatChannel, default=1}
     AttachActor {type=MovieSceneScriptingActorReferenceChannel, default="<Struct 'MovieSceneObjectBindingID' (0x8518f622aa0) {}>"}
    ]
   }]
 }
 {
  type=MovieSceneCameraCutTrack
  sections [
   {type=MovieSceneCameraCutSection, end_frame=37, start_frame=0}
   {type=MovieSceneCameraCutSection, end_frame=72, start_frame=37}
   {type=MovieSceneCameraCutSection, end_frame=193, start_frame=72}
   {type=MovieSceneCameraCutSection, end_frame=224, start_frame=193}
   {type=MovieSceneCameraCutSection, end_frame=295, start_frame=224}
   {type=MovieSceneCameraCutSection, end_frame=331, start_frame=295}
   {type=MovieSceneCameraCutSection, end_frame=402, start_frame=331}
   {type=MovieSceneCameraCutSection, end_frame=452, start_frame=402}
   {type=MovieSceneCameraCutSection, end_frame=462, start_frame=452}
   {type=MovieSceneCameraCutSection, end_frame=479, start_frame=462}
   {type=MovieSceneCameraCutSection, end_frame=489, start_frame=479}
   {type=MovieSceneCameraCutSection, end_frame=508, start_frame=489}
   {type=MovieSceneCameraCutSection, end_frame=549, start_frame=508}
  ]
 }
]
